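# Niagara system for the IgniteEgg player emote, adding short-lived dark speed lines and smoke puffs to its workout motions.
# User.WorkoutSelect indexes arrays that switch and tune one sprite and two mesh speed-line emitters, spawned around skeletal mesh bones or sockets.
# User.Running scales the smoke spawn rate, and smoke sprites use SubUV animation with curve-driven dynamic material parameters.
# Speed lines adjust camera offset for front end versus gameplay; the holo-squares emitter is disabled.
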NiagaraSystem /BRCosmetics/Animation/Game/MainPlayer/Emotes/IgniteEgg/FX/NS_Emote_IgniteEgg
dependencies {
 usedDataInterfaces [
  &NiagaraDataInterfaceCurve
  &NiagaraDataInterfaceArrayFloat
  &NiagaraDataInterfaceArrayInt32
  &NiagaraDataInterfaceArrayFloat3
  &NiagaraDataInterfaceVector2DCurve
  &NiagaraDataInterfaceSkeletalMesh
 ]
 usedDynamicInputs [&MakeFloatFromVector2d]
 usedModules [
  &SystemState
  &self:SystemSpawnScript.SystemScriptSource.SystemScriptGraph.NiagaraNodeAssignment_1.SetVariables_#1
  &self:SystemSpawnScript.SystemScriptSource.SystemScriptGraph.NiagaraNodeAssignment_2.SetVariables_#2
 ]
}
emitters {
 Em_Holo_Squares {
  bEnabled=false
  emitterUpdateScript [
   &EmitterState()
   &SpawnRate(Spawn Probability=0.5, SpawnRate=24, Use Spawn Probability=false)
  ]
  particleSpawnScript [
   &InitializeParticle(
    Color Mode=::NewEnumerator0
    Lifetime Max=1.5
    Lifetime Min=0.4
    Lifetime Mode=::NewEnumerator1
    Position Offset=<0 0 -25>
    Sprite Size Max=<6 8>
    Sprite Size Min=<2 2>
    Sprite Size Mode=::NewEnumerator2
    UsePositionOffset=true
    Write Material Random=true
   )
   &ShapeLocation(
    Cylinder Height=50
    Cylinder Radius=50
    MaskSpawnGroup=false
    Override Seed=false
    Shape Primitive=::NewEnumerator1
    Surface Only=false
    UseOwnerScale=true
   )
   &AddVelocity()
   &ApplyInitialForces()
  ]
  particleUpdateScript [
   &ParticleState()
   &ScaleSpriteSize(
    Uniform Curve Scale=0.75
    Uniform Curve Sprite Scale=DataInterface() {propertyValues {Curve {keys [{tangentMode=RCTM_User, time=0, value=0}, <at 0.25 1>, {tangentMode=RCTM_User, time=1, value=0}]}}}
   )
   &SolveForcesAndVelocity(Clamp Velocity=false, Limit Acceleration=false)
   &ScaleColor(
    Scale Alpha=&FloatFromCurve(FloatCurve=DataInterface() {propertyValues {Curve {keys [<at 0 0>,<at 0.3 1>,<at 1 0>]}}})
    Scale RGB=&Multiply_VectorByFloat(
     Vector=&MakeVectorFromLinearColorRGB(Color=$System.Color)
     Float=&FloatFromCurve(FloatCurve=DataInterface() {propertyValues {Curve {keys [<at 0 1>,<at 0.2 5>,<at 0.5 1>,<at 1 1>]}}})
    )
    ScaleA=true
    ScaleRGB=true
   )
   &ScaleColor(
    Scale Alpha=&Waveform("Waveform [1]"=::NewEnumerator0, "[1] Frequency"=0.3, "[1] Amplitude Min/Max"=<-1 1>)
    ScaleA=true
    ScaleRGB=false
   ) {enabled=false, module=ScaleColor001}
  ]
  rendererClasses [&NiagaraSpriteRendererProperties]
  renderers [{rendererClass=&NiagaraSpriteRendererProperties}]
 }
 FN_Core_Smoke_Sprite {
  emitterUpdateScript [
   &EmitterState()
   &SpawnRate(SpawnRate=&Lerp_Float(A=0, B=20, Alpha=$User.Running), Use Spawn Probability=false)
  ]
  particleSpawnScript [
   /SetVariables(Particles.InitSpriteRotRate=&RandomRangeFloat(Minimum=-2, Maximum=2)) {isSetParameters=true}
   &InitializeParticle(
    Color Maximum=<rgba 0.1255 0.1137 0.1059 1>
    Color Minimum=<rgba 0.276 0.2544 0.2327 1>
    Color Mode=::NewEnumerator2
    Lifetime Max=0.9
    Lifetime Min=0.7
    Lifetime Mode=::NewEnumerator1
    Position Offset=<5 -15 0>
    Sprite Rotation Angle=&RandomRangeFloat(Minimum=0, Maximum=360)
    Sprite Rotation Mode=::NewEnumerator0
    Sprite Size Mode=::NewEnumerator3
    Uniform Sprite Size=&MaskFloatBySpawnGroup(
     Masked Float=&RandomRangeFloat(Minimum=50, Maximum=25)
     Passthrough Float=&RandomRangeFloat(Minimum=245, Maximum=325)
    )
    UsePositionOffset=true
    Write Material Random=true
   )
   &TorusLocation(
    Align Torus With Axis=false
    Coordinate Space:ENiagaraCoordinateSpace=::World
    Large Radius=2
    MaskSpawnGroup=false
    Offset=<0 0 -10>
    Torus Distribution Mode:ENiagaraTorusDistributionMode=::NewEnumerator1
    Torus Mode=::NewEnumerator2
    U Position:NiagaraFloat=&ReturnNormalizedExecIndex(Normalized Index Includes One=false)
   ) {enabled=false}
   &ShapeLocation(
    MaskSpawnGroup=false
    Override Seed=false
    Ring Radius=12
    Rotation Coordinate Space=::World
    Shape Primitive=::NewEnumerator4
    UseOwnerScale=true
   )
   &AddVelocityFromPoint(Module.Use Falloff Distance=false, Velocity Strength=&RandomRangeFloat(Minimum=100, Maximum=200))
   &AddVelocity(Coordinate Space=::World, Velocity=&RandomRangeVector(Minimum=<0 0 175>, Maximum=<0 0 200>)) {module=AddVelocity002}
   &CameraOffset(Camera Offset Amount=35)
   /SetVariables(Particles.Velocity=&Multiply_VectorByFloat(Vector=$Particles.Velocity, Float=0.1)) {isSetParameters=true}
  ]
  particleUpdateScript [
   &Drag(Drag=1.5, Use Linear Drag=true, Use Rotational Drag=true)
   &GravityForce(Gravity=<0 0 -100>)
   &SolveForcesAndVelocity(Clamp Velocity=false, Limit Acceleration=false)
   &ParticleState()
   &SpriteRotationRate(Rotation Rate=&Multiply_Float(
     A=$Particles.InitSpriteRotRate
     B=&FloatFromCurve(
      FloatCurve=DataInterface() {propertyValues {Curve {keys [
          {arriveTangent=-4.0107, interpMode=RCIM_Cubic, leaveTangent=-4.0107, tangentMode=RCTM_User, time=-0.0046, value=1.0113}
          {arriveTangent=-0.2496, interpMode=RCIM_Cubic, leaveTangent=-0.2496, tangentMode=RCTM_User, time=0.5, value=0.16}
         ]}}}
      Scale Curve=250
     )
    ))
   &ScaleColor(
    Scale Alpha=&FloatFromCurve(FloatCurve=DataInterface() {propertyValues {Curve {keys [{interpMode=RCIM_Cubic, tangentMode=RCTM_User, time=0, value=1}, <at 1 0>]}}})
    Scale RGB=&VectorFromFloat(Value=&FloatFromCurve(FloatCurve=DataInterface() {propertyValues {Curve {keys [
          {arriveTangent=5.5, interpMode=RCIM_Cubic, leaveTangent=5.5, time=0, value=0.4}
          {arriveTangent=5.5, interpMode=RCIM_Cubic, leaveTangent=5.5, time=0.05, value=3}
          {arriveTangent=-2.6916, interpMode=RCIM_Cubic, leaveTangent=-2.6916, tangentMode=RCTM_User, time=0.2, value=1.5}
          {arriveTangent=-0.625, interpMode=RCIM_Cubic, leaveTangent=-0.625, time=1, value=1}
         ]}}}))
    ScaleA=true
    ScaleRGB=true
   )
   &ScaleSpriteSize(Scale Factor=&Vector2DFromFloat(Value=&FloatFromCurve(FloatCurve=DataInterface() {propertyValues {Curve {keys [
          {arriveTangent=8.1876, interpMode=RCIM_Cubic, leaveTangent=5.3162, tangentMode=RCTM_User, time=0, value=0.15}
          {arriveTangent=1.0025, interpMode=RCIM_Cubic, leaveTangent=1.0025, tangentMode=RCTM_User, time=0.161, value=0.7059}
          {arriveTangent=0.1638, interpMode=RCIM_Cubic, leaveTangent=0.1638, tangentMode=RCTM_User, time=1.0463, value=1}
         ]}}})))
   &DynamicMaterialParameters(
    Index 0 Param 1=&Add_Float(
     A=$Particles.RandomFloat
     B=&FloatFromCurve(FloatCurve=DataInterface() {propertyValues {Curve {keys [
          {arriveTangent=2.7533, interpMode=RCIM_Cubic, leaveTangent=2.847, tangentMode=RCTM_Break, time=0, value=0}
          {arriveTangent=1.3581, interpMode=RCIM_Cubic, leaveTangent=1.3581, tangentMode=RCTM_User, time=0.2558, value=0.5125}
          {arriveTangent=0.0529, interpMode=RCIM_Cubic, leaveTangent=0.2021, tangentMode=RCTM_User, time=1, value=1}
         ]}}})
    )
    Index 0 Param 3=&FloatFromCurve(FloatCurve=DataInterface() {propertyValues {Curve {keys [{arriveTangent=-53.9199, time=0, value=0}, {leaveTangent=0.2723, time=0.25, value=0}]}}})
    Index 0 Param 4=&FloatFromCurve(FloatCurve=DataInterface() {propertyValues {Curve {keys [<at 0 1>,<at 1 0>]}}})
    Index 1 Param 1=&FloatFromCurve(FloatCurve=DataInterface() {propertyValues {Curve {keys [<at 0 0>, {arriveTangent=4.2257, interpMode=RCIM_Cubic, leaveTangent=4.2257, tangentMode=RCTM_User, time=0.25, value=0}]}}})
    Index 1 Param 2=&FloatFromCurve(FloatCurve=DataInterface() {propertyValues {Curve {keys [<at 0 0>]}}})
    Index 1 Param 3=&FloatFromCurve(FloatCurve=DataInterface() {propertyValues {Curve {keys [{interpMode=RCIM_Cubic, leaveTangent=6.3917, tangentMode=RCTM_Break, time=0, value=0}]}}})
    Index 1 Param 4=&FloatFromCurve(FloatCurve=DataInterface() {propertyValues {Curve {keys [{arriveTangent=-0.136, leaveTangent=-0.136, time=0, value=1}]}}})
    Index 2 Param 1=&FloatFromCurve(FloatCurve=DataInterface() {propertyValues {Curve {keys [
         {arriveTangent=2.646, interpMode=RCIM_Cubic, leaveTangent=2.646, tangentMode=RCTM_User, time=0, value=0}
         {leaveTangent=-0.0719, time=1, value=1}
        ]}}})
    Index 2 Param 2=&FloatFromCurve(
     FloatCurve=DataInterface() {propertyValues {Curve {keys [
         {arriveTangent=1.9056, time=0, value=0.2}
         {arriveTangent=0.0259, interpMode=RCIM_Cubic, leaveTangent=0.0259, tangentMode=RCTM_User, time=1, value=1.4}
        ]}}}
     Scale Curve=&RandomRangeFloat(Minimum=0.5, Maximum=1)
    )
    Index 2 Param 3=&FloatFromCurve(FloatCurve=DataInterface() {propertyValues {Curve {keys [
         {interpMode=RCIM_Cubic, leaveTangent=2.3064, tangentMode=RCTM_Break, time=0, value=1}
         {interpMode=RCIM_Cubic, leaveTangent=11.1233, tangentMode=RCTM_Break, time=1, value=2}
        ]}}})
    Index 2 Param 4=&FloatFromCurve(
     FloatCurve=DataInterface() {propertyValues {Curve {keys [
         {arriveTangent=-0.6867, interpMode=RCIM_Cubic, leaveTangent=-0.6867, tangentMode=RCTM_User, time=0, value=1}
         {interpMode=RCIM_Cubic, leaveTangent=-1.2956, tangentMode=RCTM_Break, time=1, value=0.75}
        ]}}}
     Scale Curve=1.6
    )
    Module.Param0WriteEnabled=true
    Module.Param0WriteEnabled1=true
    Module.Param0WriteEnabled2=true
    Module.Param1WriteEnabled=false
    Module.Param1WriteEnabled1=true
    Module.Param1WriteEnabled2=true
    Module.Param2WriteEnabled=true
    Module.Param2WriteEnabled1=true
    Module.Param2WriteEnabled2=true
    Module.Param3WriteEnabled=true
    Module.Param3WriteEnabled1=true
    Module.Param3WriteEnabled2=true
    Write Parameter Index 1=true
    Write Parameter Index 2=true
   )
   &SubUVAnimation(
    End Frame:NiagaraInt32=&Add_Integer(A=0, B=35)
    SubUV Animation Mode:ENiagara_SubUVLookupMode=::NewEnumerator1
    SubUV Curve=DataInterface() {propertyValues {Curve {keys [
        {arriveTangent=3.375, interpMode=RCIM_Cubic, leaveTangent=3.375, time=0, value=0}
        {arriveTangent=1, interpMode=RCIM_Cubic, leaveTangent=1, time=0.2, value=0.675}
        {arriveTangent=0.4062, interpMode=RCIM_Cubic, leaveTangent=0.4062, time=1, value=1}
       ]}}}
   )
  ]
  rendererClasses [&NiagaraSpriteRendererProperties]
  renderers [{rendererClass=&NiagaraSpriteRendererProperties}]
 }
 Speed_Lines {
  emitterUpdateScript [
   &EmitterState(Life Cycle Mode=::NewEnumerator1, Loop Duration=0.95, UseLoopDelay=false)
   &SpawnRate(
    SpawnRate=&SelectFloatFromArray(
     Direct Array Index=&MakeIntFromFloat(Float=$User.WorkoutSelect)
     Float Selection Array=DataInterface() {propertyValues {FloatData [100,0,100,100,0]}}
    )
    Use Spawn Probability=false
   ) {module=SpawnRate001}
  ]
  particleSpawnScript [
   &InitializeParticle(
    Color=&RandomRangeLinearColor(Minimum=<rgba 0 0 0 10>, Maximum=<rgba 0.0521 0.0521 0.0521 15>)
    Lifetime Max=0.14
    Lifetime Min=0.095
    Lifetime Mode=::NewEnumerator1
    Position Mode=::NewEnumerator1
    Sprite Size Max=<66.6667 125>
    Sprite Size Min=<16.6667 41.6667>
    Sprite Size Mode=::NewEnumerator2
    UsePositionOffset=false
    Write Material Random=true
   )
   &SkeletalMeshLocation(
    Bone / Socket Index=&RandomRangeInt(
     Minimum=&SelectIntFromArray(
      Direct Array Index=&MakeIntFromFloat(Float=$User.WorkoutSelect)
      Int Selection Array=DataInterface() {propertyValues {IntData [0,0,0,3]}}
     )
     Maximum=&SelectIntFromArray(
      Direct Array Index=&MakeIntFromFloat(Float=$User.WorkoutSelect)
      Int Selection Array=DataInterface() {propertyValues {IntData [0,0,0,4]}}
     )
    )
    Bone Sampling Mode=::NewEnumerator4
    MaskSpawnGroup=false
    Skeletal Mesh=$System.Skeletal Mesh
   )
   &ShapeLocation(
    Box Size=&SelectVectorFromArray(
     Direct Array Index=&MakeIntFromFloat(Float=$User.WorkoutSelect)
     Vector Selection Array=DataInterface() {propertyValues {FloatData [<50 50 0>, <100 50 1>, <30 30 50>, <20 20 25>]}}
    )
    MaskSpawnGroup=false
    Offset Coordinate Space=::Local
    Offset Mode=::NewEnumerator0
    Override Seed=false
    Shape Origin=$Particles.SkeletalMeshLocation.SampledPosition
    Shape Primitive=::NewEnumerator2
    Surface Only=false
    UseOwnerScale=true
   )
   &AddVelocity(Velocity=&Multiply_Vector(
     A=&MakeVector(
      X=0
      Y=0
      Z=&SelectFloatFromArray(
       Direct Array Index=&MakeIntFromFloat(Float=$User.WorkoutSelect)
       Float Selection Array=DataInterface() {propertyValues {FloatData [100,0,0,0]}}
      )
     )
     B=&VectorFromFloat(Value=&SelectFloatFromArray(
       Array Sampling Mode=::NewEnumerator0
       Direct Array Index=?
       Float Selection Array=DataInterface() {propertyValues {FloatData [-1,1]}}
       Random Seed=0
      ))
    )) {module=AddVelocity001}
   &CameraOffset(Camera Offset Amount=&FN_FrontEndFloatModifier(Gameplay=0, Front End=-30))
  ]
  particleUpdateScript [
   &ParticleState()
   &Drag(Drag=5, Use Linear Drag=true, Use Rotational Drag=false)
   &SolveForcesAndVelocity(Clamp Velocity=false, Limit Acceleration=false)
   &ScaleColor(
    Scale Alpha=&FloatFromCurve(FloatCurve=DataInterface() {propertyValues {Curve {keys [
         {interpMode=RCIM_Cubic, tangentMode=RCTM_User, time=0.1552, value=0}
         {interpMode=RCIM_Cubic, tangentMode=RCTM_User, time=0.5, value=1}
         {tangentMode=RCTM_User, time=1, value=0}
        ]}}})
    ScaleA=true
    ScaleRGB=true
   )
   &ScaleSpriteSize(
    Non-Uniform Curve Sprite Scale=DataInterface() {propertyValues {
      XCurve {keys [{tangentMode=RCTM_User, time=0, value=0.5}, {tangentMode=RCTM_User, time=1, value=1.5}]}
      YCurve {keys [{tangentMode=RCTM_User, time=0, value=1}, {tangentMode=RCTM_User, time=1, value=0.5}]}
     }}
    Scale Sprite Size Mode=::NewEnumerator3
   )
   &ScaleSpriteSize(
    Scale Sprite Size Mode=::NewEnumerator0
    Uniform Scale Factor=&SelectFloatFromArray(
     Direct Array Index=&MakeIntFromFloat(Float=$User.WorkoutSelect)
     Float Selection Array=DataInterface() {propertyValues {FloatData [1,1,1,0.8]}}
    )
   ) {module=ScaleSpriteSize001}
   /SetVariables(Particles.New Position:NiagaraPosition=&Add_Vector(
     A=Linked() {linkedVariable {name=Particles.Position, type=NiagaraPosition}}
     B=Linked() {linkedVariable {name=Engine.Emitter.SimulationPosition, type=NiagaraPosition}}
    )) {enabled=false, isSetParameters=true, module=SetVariables_#3}
  ]
  rendererClasses [&NiagaraSpriteRendererProperties]
  renderers [{rendererClass=&NiagaraSpriteRendererProperties}]
 }
 Speed_Lines001 {
  emitterUpdateScript [
   &EmitterState(Life Cycle Mode=::NewEnumerator1, Loop Duration=0.95, UseLoopDelay=false)
   &SpawnRate(
    SpawnRate=&SelectFloatFromArray(
     Direct Array Index=&MakeIntFromFloat(Float=$User.WorkoutSelect)
     Float Selection Array=DataInterface() {propertyValues {FloatData [0,30,0,0,0]}}
    )
    Use Spawn Probability=false
   ) {module=SpawnRate001}
  ]
  particleSpawnScript [
   &InitializeParticle(
    Color=&RandomRangeLinearColor(Minimum=<rgba 0 0 0 10>, Maximum=<rgba 0.0521 0.0521 0.0521 15>)
    Lifetime Max=0.168
    Lifetime Min=0.114
    Lifetime Mode=::NewEnumerator1
    Mesh Scale Max=<4 3.4 2.5>
    Mesh Scale Min=<2.2 2.2 2.5>
    Mesh Scale Mode=::NewEnumerator2
    Position Mode=::NewEnumerator1
    UsePositionOffset=false
    Write Material Random=true
   )
   &SkeletalMeshLocation(Bone / Socket Index=5, Bone Sampling Mode=::NewEnumerator4, MaskSpawnGroup=false, Skeletal Mesh=$System.Skeletal Mesh)
   &ShapeLocation(
    Box Size=<50 45 20>
    MaskSpawnGroup=false
    Offset=<20 35 67>
    Offset Coordinate Space=::Local
    Offset Mode=::NewEnumerator0
    Override Seed=false
    Shape Origin=$Particles.SkeletalMeshLocation.SampledPosition
    Shape Primitive=::NewEnumerator2
    Surface Only=false
    UseOwnerScale=true
   )
   &AddVelocity(Velocity=&Multiply_Vector(
     A=&MakeVector(
      X=0
      Y=0
      Z=&SelectFloatFromArray(
       Direct Array Index=&MakeIntFromFloat(Float=$User.WorkoutSelect)
       Float Selection Array=DataInterface() {propertyValues {FloatData [100,0,0,100]}}
      )
     )
     B=&VectorFromFloat(Value=&SelectFloatFromArray(
       Array Sampling Mode=::NewEnumerator0
       Direct Array Index=?
       Float Selection Array=DataInterface() {propertyValues {FloatData [-1,1]}}
       Random Seed=0
      ))
    )) {enabled=false, module=AddVelocity001}
   &CameraOffset(Camera Offset Amount=&FN_FrontEndFloatModifier(Gameplay=0, Front End=-30))
  ]
  particleUpdateScript [
   &ParticleState()
   &ScaleMeshSize(Scale Factor=&MakeVector(
     X=&FloatFromCurve(FloatCurve=DataInterface() {propertyValues {Curve {keys [<at 0 0.8>,<at 1 1>]}}})
     Y=&FloatFromCurve(FloatCurve=DataInterface() {propertyValues {Curve {defaultValue=1, keys [{tangentMode=RCTM_User, time=0, value=1}, {tangentMode=RCTM_User, time=1, value=0.4}]}}})
     Z=&FloatFromCurve(FloatCurve=DataInterface() {propertyValues {Curve {defaultValue=1, keys [{tangentMode=RCTM_User, time=0, value=1}, {tangentMode=RCTM_User, time=1, value=1}]}}})
    ))
   &Drag(Drag=5, Use Linear Drag=true, Use Rotational Drag=false)
   &SolveForcesAndVelocity(Clamp Velocity=false, Limit Acceleration=false)
   &ScaleColor(
    Scale Alpha=&FloatFromCurve(FloatCurve=DataInterface() {propertyValues {Curve {keys [
         {interpMode=RCIM_Cubic, tangentMode=RCTM_User, time=0.1552, value=0}
         {interpMode=RCIM_Cubic, tangentMode=RCTM_User, time=0.5, value=1}
         {tangentMode=RCTM_User, time=1, value=0}
        ]}}})
    ScaleA=true
    ScaleRGB=true
   )
   &ScaleSpriteSize(
    Non-Uniform Curve Sprite Scale=DataInterface() {propertyValues {
      XCurve {keys [{tangentMode=RCTM_User, time=0, value=0.5}, {tangentMode=RCTM_User, time=1, value=1.5}]}
      YCurve {keys [{tangentMode=RCTM_User, time=0, value=1}, {tangentMode=RCTM_User, time=1, value=0.5}]}
     }}
    Scale Sprite Size Mode=::NewEnumerator3
   )
   /SetVariables(Particles.New Position:NiagaraPosition=&Add_Vector(
     A=Linked() {linkedVariable {name=Particles.Position, type=NiagaraPosition}}
     B=Linked() {linkedVariable {name=Engine.Emitter.SimulationPosition, type=NiagaraPosition}}
    )) {enabled=false, isSetParameters=true, module=SetVariables_#3}
  ]
  rendererClasses [&NiagaraMeshRendererProperties]
  renderers [{rendererClass=&NiagaraMeshRendererProperties, rendererIndex=0}]
 }
 Speed_Lines002 {
  emitterUpdateScript [
   &EmitterState(Life Cycle Mode=::NewEnumerator1, Loop Duration=0.95, UseLoopDelay=false)
   &SpawnRate(
    SpawnRate=&SelectFloatFromArray(
     Direct Array Index=&MakeIntFromFloat(Float=$User.WorkoutSelect)
     Float Selection Array=DataInterface() {propertyValues {FloatData [0,0,0,24,0]}}
    )
    Use Spawn Probability=false
   ) {module=SpawnRate001}
  ]
  particleSpawnScript [
   &InitializeParticle(
    Color=&RandomRangeLinearColor(Minimum=<rgba 0 0 0 20>, Maximum=<rgba 0.0521 0.0521 0.0521 25>)
    Lifetime Max=0.168
    Lifetime Min=0.114
    Lifetime Mode=::NewEnumerator1
    Mesh Scale Max=<0.8 0.8 0.5>
    Mesh Scale Min=<0.8 0.8 0.5>
    Mesh Scale Mode=::NewEnumerator2
    Position Mode=::NewEnumerator1
    UsePositionOffset=false
    Write Material Random=true
   )
   &SkeletalMeshLocation(Bone / Socket Index=5, Bone Sampling Mode=::NewEnumerator4, MaskSpawnGroup=false, Skeletal Mesh=$System.Skeletal Mesh)
   &ShapeLocation(
    Box Size=<5 5 8>
    MaskSpawnGroup=false
    Offset=&SelectVectorFromArray(
     Direct Array Index=&RandomRangeInt(Minimum=0, Maximum=1)
     Vector Selection Array=DataInterface() {propertyValues {FloatData [<-16 0 113>, <-16 -40 113>]}}
    )
    Offset Coordinate Space=::Local
    Offset Mode=::NewEnumerator0
    Override Seed=false
    Shape Origin=$Particles.SkeletalMeshLocation.SampledPosition
    Shape Primitive=::NewEnumerator2
    Surface Only=false
    UseOwnerScale=true
   )
   &AddVelocity(Velocity=&Multiply_Vector(
     A=&MakeVector(
      X=0
      Y=0
      Z=&SelectFloatFromArray(
       Direct Array Index=&MakeIntFromFloat(Float=$User.WorkoutSelect)
       Float Selection Array=DataInterface() {propertyValues {FloatData [100,0,0,100]}}
      )
     )
     B=&VectorFromFloat(Value=&SelectFloatFromArray(
       Array Sampling Mode=::NewEnumerator0
       Direct Array Index=?
       Float Selection Array=DataInterface() {propertyValues {FloatData [-1,1]}}
       Random Seed=0
      ))
    )) {enabled=false, module=AddVelocity001}
   &CameraOffset(Camera Offset Amount=&FN_FrontEndFloatModifier(Gameplay=0, Front End=-30))
  ]
  particleUpdateScript [
   &ParticleState()
   &ScaleMeshSize(Scale Factor=&MakeVector(
     X=&FloatFromCurve(FloatCurve=DataInterface() {propertyValues {Curve {keys [<at 0 0.4>,<at 1 1>]}}})
     Y=&FloatFromCurve(FloatCurve=DataInterface() {propertyValues {Curve {defaultValue=1, keys [{tangentMode=RCTM_User, time=0, value=1}, {tangentMode=RCTM_User, time=1, value=0.4}]}}})
     Z=&FloatFromCurve(FloatCurve=DataInterface() {propertyValues {Curve {defaultValue=1, keys [{tangentMode=RCTM_User, time=0, value=1}, {tangentMode=RCTM_User, time=1, value=1}]}}})
    ))
   &Drag(Drag=5, Use Linear Drag=true, Use Rotational Drag=false)
   &SolveForcesAndVelocity(Clamp Velocity=false, Limit Acceleration=false)
   &ScaleColor(
    Scale Alpha=&FloatFromCurve(FloatCurve=DataInterface() {propertyValues {Curve {keys [
         {interpMode=RCIM_Cubic, tangentMode=RCTM_User, time=0.1552, value=0}
         {interpMode=RCIM_Cubic, tangentMode=RCTM_User, time=0.5, value=1}
         {tangentMode=RCTM_User, time=1, value=0}
        ]}}})
    ScaleA=true
    ScaleRGB=true
   )
   &ScaleSpriteSize(
    Non-Uniform Curve Sprite Scale=DataInterface() {propertyValues {
      XCurve {keys [{tangentMode=RCTM_User, time=0, value=0.5}, {tangentMode=RCTM_User, time=1, value=1.5}]}
      YCurve {keys [{tangentMode=RCTM_User, time=0, value=1}, {tangentMode=RCTM_User, time=1, value=0.5}]}
     }}
    Scale Sprite Size Mode=::NewEnumerator3
   )
   /SetVariables(Particles.New Position:NiagaraPosition=&Add_Vector(
     A=Linked() {linkedVariable {name=Particles.Position, type=NiagaraPosition}}
     B=Linked() {linkedVariable {name=Engine.Emitter.SimulationPosition, type=NiagaraPosition}}
    )) {enabled=false, isSetParameters=true, module=SetVariables_#3}
  ]
  rendererClasses [&NiagaraMeshRendererProperties]
  renderers [{rendererClass=&NiagaraMeshRendererProperties, rendererIndex=0}]
 }
}
summary {
 systemName=NS_Emote_IgniteEgg
 emitters [
  {bEnabled=false, emitterName=Em_Holo_Squares, rendererClasses [&NiagaraSpriteRendererProperties]}
  {emitterName=Speed_Lines, rendererClasses [&NiagaraSpriteRendererProperties]}
  {emitterName=Speed_Lines001, rendererClasses [&NiagaraMeshRendererProperties]}
  {emitterName=Speed_Lines002, rendererClasses [&NiagaraMeshRendererProperties]}
  {emitterName=FN_Core_Smoke_Sprite, rendererClasses [&NiagaraSpriteRendererProperties]}
 ]
 userVariables [
  User.bIsFrontEnd {type=NiagaraBool}
  User.bIsFrontEndPreview {type=NiagaraBool}
  User.HologramColorLerp {type=NiagaraFloat, defaultValue=0, description=""}
  User.Is WeakPoint {type=NiagaraBool, defaultValue=true}
  User.Running {type=NiagaraFloat, defaultValue=0, description=""}
  User.WorkoutSelect {type=NiagaraFloat, defaultValue=0, description=""}
 ]
}
