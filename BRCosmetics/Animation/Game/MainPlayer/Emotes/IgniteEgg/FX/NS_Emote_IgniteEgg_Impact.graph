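# Niagara impact effect for the IgniteEgg emote, built from two active burst emitters.
# Impact_Flare_and_Corona001 spawns one short-lived particle drawn by two sprite renderers and a mesh, animated through curve-driven dynamic material parameters, scale and fade.
# Spiky_Burst adds a randomly mirrored, fading mesh gated by User.Is WeakPoint (default true), using a Pixelate dynamic input to step its growth.
# Holographic squares and debug gnomon emitters are disabled.
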NiagaraSystem /BRCosmetics/Animation/Game/MainPlayer/Emotes/IgniteEgg/FX/NS_Emote_IgniteEgg_Impact
dependencies {
 usedDataInterfaces [&NiagaraDataInterfaceCurve, &NiagaraDataInterfaceSkeletalMesh, &NiagaraDataInterfaceArrayFloat]
 usedDynamicInputs [&SelectFloatFromArray, &MakeIntFromFloat]
 usedModules [
  &SystemState
  &self:SystemSpawnScript.SystemScriptSource.SystemScriptGraph.NiagaraNodeAssignment_1.SetVariables_#1
  &self:SystemSpawnScript.SystemScriptSource.SystemScriptGraph.NiagaraNodeAssignment_2.SetVariables_#2
 ]
}
emitters {
 Em_Holo_Squares {
  bEnabled=false
  emitterUpdateScript [
   &EmitterState()
   &SpawnRate(Spawn Probability=0.5, SpawnRate=24, Use Spawn Probability=false)
  ]
  particleSpawnScript [
   &InitializeParticle(
    Color Mode=::NewEnumerator0
    Lifetime Max=1.5
    Lifetime Min=0.4
    Lifetime Mode=::NewEnumerator1
    Position Offset=<0 0 -25>
    Sprite Size Max=<6 8>
    Sprite Size Min=<2 2>
    Sprite Size Mode=::NewEnumerator2
    UsePositionOffset=true
    Write Material Random=true
   )
   &ShapeLocation(
    Cylinder Height=50
    Cylinder Radius=50
    MaskSpawnGroup=false
    Override Seed=false
    Shape Primitive=::NewEnumerator1
    Surface Only=false
    UseOwnerScale=true
   )
   &AddVelocity()
   &ApplyInitialForces()
  ]
  particleUpdateScript [
   &ParticleState()
   &ScaleSpriteSize(
    Uniform Curve Scale=0.75
    Uniform Curve Sprite Scale=DataInterface() {propertyValues {Curve {keys [{tangentMode=RCTM_User, time=0, value=0}, <at 0.25 1>, {tangentMode=RCTM_User, time=1, value=0}]}}}
   )
   &SolveForcesAndVelocity(Clamp Velocity=false, Limit Acceleration=false)
   &ScaleColor(
    Scale Alpha=&FloatFromCurve(FloatCurve=DataInterface() {propertyValues {Curve {keys [<at 0 0>,<at 0.3 1>,<at 1 0>]}}})
    Scale RGB=&Multiply_VectorByFloat(
     Vector=&MakeVectorFromLinearColorRGB(Color=$System.Color)
     Float=&FloatFromCurve(FloatCurve=DataInterface() {propertyValues {Curve {keys [<at 0 1>,<at 0.2 5>,<at 0.5 1>,<at 1 1>]}}})
    )
    ScaleA=true
    ScaleRGB=true
   )
   &ScaleColor(
    Scale Alpha=&Waveform("Waveform [1]"=::NewEnumerator0, "[1] Frequency"=0.3, "[1] Amplitude Min/Max"=<-1 1>)
    ScaleA=true
    ScaleRGB=false
   ) {enabled=false, module=ScaleColor001}
  ]
  rendererClasses [&NiagaraSpriteRendererProperties]
  renderers [{rendererClass=&NiagaraSpriteRendererProperties}]
 }
 FN_DebugGnomon {
  bEnabled=false
  emitterUpdateScript [
   &EmitterState(
    Life Cycle Mode=::NewEnumerator1
    Loop Behavior=::NewEnumerator1
    Loop Duration Mode=::NewEnumerator1
    Scalability Mode=::NewEnumerator1
    UseLoopDelay=false
   )
   &SpawnBurst_Instantaneous(Use Spawn Probability=false, UseLoopCountLimit=false)
  ]
  particleSpawnScript [&InitializeParticle(
    Color Mode=::NewEnumerator0
    Lifetime=5
    Mesh Scale Mode=::NewEnumerator3
    Mesh Uniform Scale=5
    UsePositionOffset=false
    Write Material Random=true
   )]
  particleUpdateScript [&ParticleState(Kill Particles When Lifetime Has Elapsed=false)]
  rendererClasses [&NiagaraMeshRendererProperties]
  renderers [{rendererClass=&NiagaraMeshRendererProperties, rendererIndex=0}]
 }
 Impact_Flare_and_Corona001 {
  emitterUpdateScript [
   &EmitterState()
   &SpawnBurst_Instantaneous(Spawn Count=1, Use Spawn Probability=false)
  ]
  particleSpawnScript [
   /Niagara/Modules/Spawn/Initialization/InitializeParticle(
    Color=<rgba 1 1 1 0.05>
    Lifetime=0.2
    Mesh Scale=&UniformRangedVector(Minimum=<0.1 1.1111 1.1111>, Maximum=<0.1 1.25 1.25>)
    Module.Write Color=true
    Module.Write Lifetime=true
    Module.Write Mass=true
    Module.Write Material Random=true
    Module.Write Position=false
    Module.Write Scale=true
    Module.Write SpriteRotation=false
    Module.Write SpriteSize=true
    Position=&SimulationPosition()
    Sprite Size=<133.3333 116.6667>
   )
   /SetVariables(Particles.CoronaSpriteSize=&Vector2DFromFloat(Value=120)) {isSetParameters=true}
   /SetVariables(Particles.ImpactColor:LinearColor=&MakeLinearColorFromVectorAndFloat("Vector (RGB)"=<0.5 0.5 0.5>, "Float (Alpha)"=&Multiply_Float(A=1, B=1))) {isSetParameters=true}
   &InitialMeshOrientation(Orientation Vector=<0 0 1>, Use Orientation Vector=true, Use Rotation Vector=true)
   &ApplyInitialForces()
   &AlignSpriteToMeshOrientation(SetAlignment=true, SetFacing=true)
  ]
  particleUpdateScript [
   &ParticleState()
   &DynamicMaterialParameters(
    Index 0 Param 1=&FloatFromCurve(FloatCurve=DataInterface() {propertyValues {Curve {keys [<at 0 1 RCIM_Cubic>,<at 0.8 0 RCIM_Cubic>]}}})
    Index 0 Param 2=2
    Index 0 Param 3=1
    Index 0 Param 4=1
    Index 1 Param 1=&FloatFromCurve(FloatCurve=DataInterface() {propertyValues {Curve {keys [
         {arriveTangent=-2.3163, interpMode=RCIM_Cubic, time=0, value=2}
         {arriveTangent=-1.9019, interpMode=RCIM_Cubic, leaveTangent=-1.9019, tangentMode=RCTM_User, time=1, value=0.4}
        ]}}})
    Index 1 Param 2=0.4
    Index 1 Param 3=0.6
    Index 1 Param 4=1
    Index 2 Param 1=&FloatFromCurve(FloatCurve=DataInterface() {propertyValues {Curve {keys [<at 0.46 0>,<at 1 1>]}}})
    Index 2 Param 2=45
    Module.Param0WriteEnabled=true
    Module.Param0WriteEnabled1=true
    Module.Param0WriteEnabled2=true
    Module.Param1WriteEnabled=true
    Module.Param1WriteEnabled1=true
    Module.Param1WriteEnabled2=true
    Module.Param2WriteEnabled=true
    Module.Param2WriteEnabled1=true
    Module.Param2WriteEnabled2=true
    Module.Param3WriteEnabled=true
    Module.Param3WriteEnabled1=true
    Module.Param3WriteEnabled2=true
    Write Parameter Index 1=true
    Write Parameter Index 2=true
   )
   /SetVariables(Particles.CoronaSpriteColor:LinearColor=&MakeLinearColorFromVectorAndFloat(
     "Vector (RGB)"=&MakeVectorFromLinearColorRGB(Color=<rgba 0.5521 0.5521 0.5521 0.1>)
     "Float (Alpha)"=&Multiply_Float(
      A=&FloatFromCurve(FloatCurve=DataInterface() {propertyValues {Curve {keys [<at 0.35 1.5 RCIM_Cubic>,<at 0.95 0>]}}})
      B=1
     )
    )) {isSetParameters=true}
   /SetVariables() {isSetParameters=true}
   &ScaleColor(
    Scale Alpha=&FloatFromCurve(FloatCurve=DataInterface() {propertyValues {Curve {keys [<at 0 1>,<at 1 0>]}}})
    Scale RGB=&MakeVectorFromLinearColorRGB(Color=<rgba 1 1 1 1>)
    ScaleA=true
    ScaleRGB=true
   )
   &ScaleMeshSize(Scale Factor=&VectorFromFloat(Value=&FloatFromCurve(FloatCurve=DataInterface() {propertyValues {Curve {keys [<at 0 0.294 RCIM_Cubic>,<at 1 1.05>]}}}, Scale Curve=1.3)))
  ]
  rendererClasses [&NiagaraSpriteRendererProperties, &NiagaraMeshRendererProperties]
  renderers [
   {rendererClass=&NiagaraSpriteRendererProperties}
   {rendererClass=&NiagaraSpriteRendererProperties, rendererIndex=1}
   {rendererClass=&NiagaraMeshRendererProperties, rendererIndex=2}
  ]
 }
 Spiky_Burst {
  emitterUpdateScript [
   &EmitterState(Life Cycle Mode=::NewEnumerator1, Loop Behavior=::NewEnumerator1, Loop Duration=1, UseLoopDelay=false)
   &SpawnBurst_Instantaneous(
    Spawn Probability=&MakeCustomFloatFromBool(Bool=$User.Is WeakPoint, True Float=1, False Float=0)
    Use Spawn Probability=true
    UseLoopCountLimit=false
   )
  ]
  particleSpawnScript [
   /Niagara/Modules/Spawn/Initialization/V2/InitializeParticle(
    Color=<rgba 0.8085 0.9286 1 0.02>
    Lifetime Max=0.35
    Lifetime Min=0.3
    Lifetime Mode=::NewEnumerator1
    Mesh Scale=&Multiply_VectorByFloat(
     Vector=&MakeVector(X=&UniformAOrBFloat(A=-1), Y=&UniformAOrBFloat(A=-1), Z=1)
     Float=0.2
    )
    Mesh Scale Mode=::NewEnumerator4
    Position Offset=<-2 0 1>
    Sprite Rotation Mode=::NewEnumerator2
    Sprite Size Mode=::NewEnumerator1
    Uniform Sprite Size Max=1
    Uniform Sprite Size Min=1
    UsePositionOffset=true
    Write Material Random=true
   )
   &InitialMeshOrientation(Mesh Orientation Mode=::NewEnumerator2, Orientation Vector=<0 0 1>, Use Rotation Vector=true)
   &CameraOffset(Camera Offset Amount=25) {enabled=false}
   &ApplyInitialForces()
   &AlignSpriteToMeshOrientation(Mesh Orientation Relative Sprite Facing Vector:Vector3f=<0 0 1>, SetAlignment=true, SetFacing=true)
  ]
  particleUpdateScript [
   &ParticleState()
   &ScaleMeshSize(Scale Factor=&VectorFromFloat(Value=&self:Pixelate_01(
      Input:NiagaraFloat=&FloatFromCurve(FloatCurve=DataInterface() {propertyValues {Curve {keys [
           {arriveTangent=1.1217, interpMode=RCIM_Cubic, leaveTangent=1.1217, tangentMode=RCTM_User, time=0, value=0}
           {arriveTangent=0.4219, interpMode=RCIM_Cubic, leaveTangent=0.4219, tangentMode=RCTM_User, time=1, value=1}
          ]}}})
      Pixel Value:NiagaraFloat=8
     )))
   &Drag(Drag=0.75, Use Linear Drag=true, Use Rotational Drag=false)
   &ScaleColor(
    Scale Alpha=&FloatFromCurve(
     FloatCurve=DataInterface() {propertyValues {Curve {keys [
         {arriveTangent=2.0249, interpMode=RCIM_Cubic, leaveTangent=2.0249, tangentMode=RCTM_User, time=0, value=0}
         {
          arriveTangentWeight=0.2468
          interpMode=RCIM_Cubic
          leaveTangent=-6.1069
          leaveTangentWeight=1.0963
          tangentMode=RCTM_Break
          tangentWeightMode=RCTWM_WeightedBoth
          time=0.35
          value=1
         }
         {interpMode=RCIM_Cubic, leaveTangent=-2.393, time=1, value=0}
        ]}}}
     Scale Curve=0.25
    )
    ScaleA=true
    ScaleRGB=true
   )
   &SolveForcesAndVelocity(Clamp Velocity=false, Limit Acceleration=false)
  ]
  rendererClasses [&NiagaraMeshRendererProperties]
  renderers [{rendererClass=&NiagaraMeshRendererProperties, rendererIndex=0}]
 }
}
summary {
 systemName=NS_Emote_IgniteEgg_Impact
 emitters [
  {bEnabled=false, emitterName=Em_Holo_Squares, rendererClasses [&NiagaraSpriteRendererProperties]}
  {emitterName=Spiky_Burst, rendererClasses [&NiagaraMeshRendererProperties]}
  {bEnabled=false, emitterName=FN_DebugGnomon, rendererClasses [&NiagaraMeshRendererProperties]}
  {
   emitterName=Impact_Flare_and_Corona001
   rendererClasses [&NiagaraSpriteRendererProperties, &NiagaraMeshRendererProperties]
  }
 ]
 userVariables [
  User.bIsFrontEnd {type=NiagaraBool}
  User.bIsFrontEndPreview {type=NiagaraBool}
  User.HologramColorLerp {type=NiagaraFloat, defaultValue=0, description=""}
  User.Is WeakPoint {type=NiagaraBool, defaultValue=true}
  User.WorkoutSelect {type=NiagaraFloat, defaultValue=0, description=""}
 ]
}
