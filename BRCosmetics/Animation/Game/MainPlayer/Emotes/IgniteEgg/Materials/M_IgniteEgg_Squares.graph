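# Unlit additive Niagara sprite and mesh particle material that draws glowing squares with scanlines for the IgniteEgg emote.
# Two UV-based square patterns, one inverted at a different scale, sum into a mask that multiplies particle color and alpha.
# Animated screen-aligned scanlines, blended in by an intensity parameter and shared through a named reroute, modulate both color and alpha, with color passing through the shared emissive-adjustment function.
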Material /BRCosmetics/Animation/Game/MainPlayer/Emotes/IgniteEgg/Materials/M_IgniteEgg_Squares
graph {
 properties {
  bAutomaticallySetUsageInEditor=false
  bUsedWithNiagaraMeshParticles=true
  bUsedWithNiagaraSprites=true
  blendMode=BLEND_Additive
  shadingModel=MSM_Unlit
  usedShadingModels=Unlit
 }
 nodes [
  n1 = ParticleColor()
  n2 = TextureCoordinate()
  n3 = Constant() {r=1}
  n4 = Constant() {r=1}
  n5 = Constant() {r=0.4}
  n6 = Add(
   A=Multiply(A=MaterialFunctionCall!(UVs=@n2, Size=@n3, SizeScale) {materialFunction=/Game/Packages/Fortress_SharedMaterials/Base_Material_Functions/SquarePatternFromUVs}.PatternOutput) {constB=0.9}
   B=OneMinus(None=MaterialFunctionCall(UVs=@n2, Size=@n4, SizeScale=@n5) {materialFunction=&SquarePatternFromUVs}.PatternOutput)
  )
  n7 = NamedRerouteUsage()
  MaterialFunctionCall(Emissive=Multiply(A=Multiply(A=@n1.RGB, B=@n6), B=@n7)) {materialFunction=&MF_FX_EmissiveAdjustments}
  Multiply(A=@n7, B=Multiply(A=@n1.A, B=@n6))
  n8 = Constant() {r=0.004}
  n9 = Time()
  n10 = MaterialFunctionCall!(
   Number of Scanlines=ScalarParameter() {defaultValue=150, parameterName=ScanlinesAmount}
   Scanline Speed=@n8
   "Scanlines Thickness (Bias)"
   Scanlines Dark Value=Constant() {r=0.5}
   Scanlines Light Value
   Block Opacity=Constant() {r=0.2}
   UVs=MaterialFunctionCall() {materialFunction=&ScreenAlignedUVs}."X 100%, Y 100%"
   Time=Multiply(A=Multiply(A=@n9, B=@n8)) {constB=20}
   Block Repeat=Constant() {r=3}
  ) {materialFunction=/Game/Effects/Fort_Effects/Effects/Atlas/Materials/MF_Atlas_Scanlines_01}
  n11 = Multiply(A=@n10.Secondary Lines, B=Constant() {r=1})
  NamedRerouteDeclaration(None=LinearInterpolate(B=Add(A=@n10.Result, B=@n11), Alpha=ScalarParameter() {defaultValue=1, group=IgniteEgg, parameterName=IE_Scanline2_Intensity}) {constA=1}) {name=Scanlines, nodeColor=<rgba 0.4062 0 1 1>}
 ]
}
